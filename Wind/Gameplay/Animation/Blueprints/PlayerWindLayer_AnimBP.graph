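# Player animation layer Blueprint for Wind gameplay that supplies a slope-sliding pose through the MovementMode layer group.
# LocomotionAdditives switches between its input pose and the WindSliding layer using bApplyWindSlideAdditive; the other MovementMode layers are unwired stubs.
# WindSliding blends rifle, pistol and no-weapon blendspaces by weapon animation type, all driven by slope roll and pitch in one sync group.
# A centre-slide additive is applied on top, limited by an LOD threshold.
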AnimBlueprint /Wind/Gameplay/Animation/Blueprints/PlayerWindLayer_AnimBP
target_skeleton=&Fortnite_M_Avg_Player_Skeleton
graphs [
 AnimGraph {nodes [Output Pose() {layerGroup=None}]}
 EventGraph {nodes [
   TryGetPawnOwner()
   Event BlueprintUpdateAnimation()
  ]}
 MovementMode_AimOffset {nodes [
   Input Pose() {name=InAimOffsetsPose}
   Input Pose() {name=InSourcePosePreAimOffsets}
   Output Pose() {layerGroup=MovementMode}
  ]}
 MovementMode_BaseLocomotion {nodes [
   Input Pose() {name=InPoseBaseLocomotion}
   Output Pose() {layerGroup=MovementMode}
  ]}
 MovementMode_FinalPoseOverride {nodes [
   Input Pose() {name=InFinalPose}
   Output Pose() {layerGroup=MovementMode}
  ]}
 MovementMode_FullBodyOverride {nodes [
   Input Pose() {name=InputPoseFullBody}
   Output Pose() {layerGroup=MovementMode}
  ]}
 MovementMode_InAirOverride {nodes [
   Input Pose() {name=InSource_Pose}
   Input Pose() {name=InAirPose}
   Output Pose() {layerGroup=MovementMode}
  ]}
 MovementMode_LocomotionAdditives {nodes [Output Pose(Result=Blend Poses by bool(
     BlendPose_0=WindSliding!() -> (Pose:Pose Link Structure) {is=title AnimGraphNode_LinkedAnimLayer, layer=WindSliding}
     BlendPose_1=Input Pose() {name=InLocomotionAdditivesPose}
     bActiveValue=Get bApplyWindSlideAdditive() -> (bApplyWindSlideAdditive:bool)
    ) {blendType=Linear}) {layerGroup=MovementMode}]}
 MovementMode_LowerBodyOverride {nodes [
   Input Pose() {name=InLowerBodyPose}
   Input Pose() {name=InSourcePose}
   Output Pose() {layerGroup=MovementMode}
  ]}
 MovementMode_OffScreen {nodes [
   Input Pose() {name=InPoseOffscreen}
   Output Pose() {layerGroup=MovementMode}
  ]}
 MovementMode_SourcePose {nodes [
   Input Pose() {name=InSourcePoseBase}
   Output Pose() {layerGroup=MovementMode}
  ]}
 WindSliding {nodes [
   n1 = Get SlopeSlidingPitch() -> (SlopeSlidingPitch:float)
   n2 = Get SlopeSlidingRoll() -> (SlopeSlidingRoll:float)
   n3 = Blendspace Player 'SlopeSliding_NoWep_BS'(X=@n2, Y=@n1) {
    blendSpace=/Game/Animation/Game/MainPlayer/Locomotion/Medium/Male/Sliding/SlopeSliding_NoWep_BS
    groupName=SlopeSliding
    groupRole=AlwaysFollower
    method=SyncGroup
   }
   n4 = Blendspace Player 'SlopeSliding_AR_BS'(X=@n2, Y=@n1) {blendSpace=/Game/Animation/Game/MainPlayer/Locomotion/Medium/Male/Sliding/SlopeSliding_AR_BS, groupName=SlopeSliding, method=SyncGroup}
   n5 = Blendspace Player 'SlopeSliding_NoWep_BS'(X=@n2, Y=@n1) {
    blendSpace=&SlopeSliding_NoWep_BS
    groupName=SlopeSliding
    groupRole=AlwaysFollower
    method=SyncGroup
   }
   n6 = Blendspace Player 'SlopeSliding_Pistol_BS'(X=@n2, Y=@n1) {
    blendSpace=/Game/Animation/Game/MainPlayer/Locomotion/Medium/Male/Sliding/SlopeSliding_Pistol_BS
    groupName=SlopeSliding
    groupRole=AlwaysFollower
    method=SyncGroup
   }
   n7 = "Blend Poses (EFortWeaponCoreAnimation)"!(
    BlendPose_0:Pose Link Structure=@n4
    BlendPose_1:Pose Link Structure=@n6
    BlendPose_2:Pose Link Structure=@n3
    BlendPose_3:Pose Link Structure=@n5
    BlendTime_0:float=.2
    BlendTime_1:float=.2
    BlendTime_2:float=0.1
    BlendTime_3:float=0.1
    ActiveEnumValue:Byte=Get WeaponCoreAnim!() -> (WeaponCoreAnim)
   ) -> (Pose:Pose Link Structure) {
    is=title AnimGraphNode_BlendListByEnum
    blendPose [{linkId=-1, sourceLinkId=-1}, {linkId=-1, sourceLinkId=-1}, {linkId=-1, sourceLinkId=-1}, {linkId=-1, sourceLinkId=-1}]
    blendTime [0.1,0.1,0.1,0.1]
   }
   Output Pose(Result=Apply Additive(
     Base=@n7
     Additive="Sequence Player (additive) 'Slide_Center_CMM_Additive'"() {sequence=/Game/Animation/Game/MainPlayer/Locomotion/Medium/Male/Sliding/Poses/Slide_Center_CMM_Additive}
    ) {lODThreshold=1, alphaBoolBlend {blendOption=Linear}}) {layerGroup=None}
  ]}
]
variables []
